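; Looping gameplay cue for WildEstate low-gravity VFX that gates particle spawning by the audio enclosure value.
; Each tick it reads the AudioWorldization Enclosure modulation bus and converts a below-0.9 test to 1 or 0.
; That value sets the "Spawn" parameter on the first particle component from the looping-start event, reusing audio worldization data to drive visuals.
; It also prints the value every tick; the overlap event and construction script are empty.
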(event EventTick
  (params (DeltaSeconds :type Float%20%28single-precision%29))
  (bind _returnvalue (Audio|Modulation|GetModulatorValue 0 "/AudioWorldization/Modulation/AudioWorldization_Bus_Enclosure.AudioWorldization_Bus_Enclosure"))
  (bind _returnvalue_2 (< _returnvalue 0.9))
  (bind _returnvalue_3 (Math|Conversions|ToFloat(Boolean) _returnvalue_2))
  (Effects|Components|ParticleSystem|SetFloatParameter (Utilities|Array|Get(acopy) (stale EventOnLoopingStart(EitherVFX-preferthisversion) ParticleComponents)) "Spawn" _returnvalue_3)
  (Development|PrintString 0 (Utilities|String|ToString(Float) _returnvalue_3)))

(event EventOnLoopingStart(EitherVFX-preferthisversion)
  (params (MyTarget :type Actor%20Object%20Reference) (Parameters :type Gameplay%20Cue%20Parameters%20Structure%20%28by%20ref%29) (ParticleComponents :type Array%20of%20FXSystem%20Component%20Object%20References) (AudioComponents :type Array%20of%20Audio%20Component%20Object%20References)))

(event Collision|EventActorBeginOverlap
  (params (OtherActor :type Actor%20Object%20Reference)))

(fn ConstructionScript)
